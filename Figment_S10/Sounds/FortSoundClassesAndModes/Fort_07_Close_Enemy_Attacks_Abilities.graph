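# Sound class for close-range enemy attack and ability sounds in Figment_S10.
# It sits under the in-game SFX class and parents a Character child class.
# It sends a small share of its signal to the LFE channel through an LFE bleed setting.
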SoundClass /Figment_S10/Sounds/FortSoundClassesAndModes/Fort_07_Close_Enemy_Attacks_Abilities
parent=&Fort_SFX_InGame
children [&/Character]
properties {properties {lFEBleed=0.25}}
