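# Sound class for enemy attack sounds in the Figment_S10 content, parented to the Character class.
# It boosts the volume of its sounds and has effects application enabled.
# A passive sound mix modifier automatically activates the shared Fort_EnemiesAreAttacking_MixMod mix while a sound in this class plays above a minimum volume threshold, so attacks can alter the mix without explicit triggering.
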SoundClass /Figment_S10/Sounds/FortSoundClassesAndModes/Fort_Enemy_Attack_PSM
parent=&/Character
properties {
 passiveSoundMixModifiers [{minVolumeThreshold=0.35, soundMix=/Game/Sounds/FortSoundClassesAndModes/Fort_EnemiesAreAttacking_MixMod}]
 properties {bApplyEffects=true, volume=1.5}
}
