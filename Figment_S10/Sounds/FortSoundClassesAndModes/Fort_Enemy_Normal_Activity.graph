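# Sound class for normal enemy activity audio in the Figment_S10 sound class hierarchy.
# It is parented to the Character sound class.
# It overrides the LFE bleed level, sending a small share of these sounds to the low-frequency channel.
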SoundClass /Figment_S10/Sounds/FortSoundClassesAndModes/Fort_Enemy_Normal_Activity
parent=&/Character
properties {properties {lFEBleed=0.2}}
