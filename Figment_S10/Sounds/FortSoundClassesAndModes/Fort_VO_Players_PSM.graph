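# Sound class for player voice-over in the Figment_S10 content, parented to the shared Fort Voice class.
# Acts as a passive mix trigger: sounds in this class automatically activate the Fort_VO_Players_MixMod sound mix while their volume is above a low threshold.
# It also enables effects and sends part of the voice to the center channel.
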SoundClass /Figment_S10/Sounds/FortSoundClassesAndModes/Fort_VO_Players_PSM
parent=/Game/Sounds/FortSoundClassesAndModes/Voice
properties {
 passiveSoundMixModifiers [{minVolumeThreshold=0.15, soundMix=/Game/Sounds/FortSoundClassesAndModes/Fort_VO_Players_MixMod}]
 properties {bApplyEffects=true, voiceCenterChannelVolume=0.5}
}
