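# Sound class for in-building ambience in Figment_S10, a child of the Foundations_Ambience class.
# A passive sound mix modifier activates the Foundations_DuckAmbience_InBuilding mix while sounds of this class play within a set volume threshold range.
# The class applies effects and sets an LFE bleed.
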SoundClass /Figment_S10/Sounds/FortSoundClassesAndModes/Foundations_Ambience_InBuilding_PSM
parent=&/Foundations_Ambience
properties {
 passiveSoundMixModifiers [{maxVolumeThreshold=1000, minVolumeThreshold=0.5, soundMix=&/Foundations_DuckAmbience_InBuilding}]
 properties {bApplyEffects=true, lFEBleed=0.2}
}
