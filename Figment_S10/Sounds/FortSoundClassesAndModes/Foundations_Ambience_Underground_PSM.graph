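# Sound class for underground ambience in Figment_S10, parented to the Foundations_Ambience class.
# Its passive sound mix modifier triggers the Foundations_DuckAmbience_Underground mix while sounds in this class play within a set volume threshold range, so activation follows playback automatically.
# The class also enables effects and sets some LFE bleed.
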SoundClass /Figment_S10/Sounds/FortSoundClassesAndModes/Foundations_Ambience_Underground_PSM
parent=&/Foundations_Ambience
properties {
 passiveSoundMixModifiers [{maxVolumeThreshold=1000, minVolumeThreshold=0.05, soundMix=&/Foundations_DuckAmbience_Underground}]
 properties {bApplyEffects=true, lFEBleed=0.3}
}
