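# Sound class for Ray's announcement voice-over in Figment S10, judging by its name, with a boosted volume.
# It is parented to the Fort_02_Critical_Announcements sound class.
# A passive sound mix modifier triggers Fort_VO_Announcement_MixMod while sounds in this class play above a low volume threshold, so the mix applies automatically during announcements.
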SoundClass /Figment_S10/Sounds/FortSoundClassesAndModes/Fort_VO_Ray_Announcement_PSM
parent=&/Fort_02_Critical_Announcements
properties {
 passiveSoundMixModifiers [{minVolumeThreshold=0.1, soundMix=&/Fort_VO_Announcement_MixMod}]
 properties {volume=1.35}
}
