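# Sound class for critical announcements in the Figment_S10 content, sitting under the Voice class in the Fortnite sound class hierarchy.
# It groups two child classes, Fort_VO_Announcement_PSM and Fort_VO_Ray_Announcement_PSM, for general and Ray voice-over announcements.
# Its only changed property is LFE bleed, which sets how much of the sound feeds the low-frequency channel.
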SoundClass /Figment_S10/Sounds/FortSoundClassesAndModes/Fort_02_Critical_Announcements
parent=/Game/Sounds/FortSoundClassesAndModes/Voice
children [
 &/Fort_VO_Announcement_PSM
 &/Fort_VO_Ray_Announcement_PSM
]
properties {properties {lFEBleed=0.25}}
